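# Niagara system for a deploy effect on the Headcase male VinderTech glider, bursting short-lived liquid-style sprites: big splats, splashes, drops and slurpy streaks.
# All four emitters sample User.SkeletalMesh, then spawn in spheres offset from that point with a shared push along negative Y.
# Curl noise and gravity move the particles, while curves drive size and alpha fades.
# Splat, splash and streak emitters also animate dynamic material parameters.
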NiagaraSystem /BRCosmetics/Gadgets/Assets/VinderTech_GliderChute/Glider_Headcase_Male/FX/NS_Glider_Accumulate_Deploy_02_FE
dependencies {usedDataInterfaces [&NiagaraDataInterfaceCurve], usedDynamicInputs [], usedModules [&SystemState]}
emitters {
 BigSplat {
  emitterUpdateScript [
   &EmitterState()
   &SpawnBurst_Instantaneous(Spawn Count=&RandomRangeInt(Minimum=5, Maximum=8), Use Spawn Probability=false)
  ]
  particleSpawnScript [
   /SetVariables(Particles.SpriteRotationRate=&RandomRangeFloat(Minimum=-0.1, Maximum=0.1, Evaluation Type=::NewEnumerator1)) {isSetParameters=true, module=SetVariables_#1}
   /Niagara/Modules/Spawn/Initialization/InitializeParticle(
    Lifetime=&RandomRangeFloat(Minimum=0.4, Maximum=0.7, Evaluation Type=::NewEnumerator1)
    Module.Write Color=true
    Module.Write Lifetime=true
    Module.Write Mass=true
    Module.Write Material Random=true
    Module.Write Position=false
    Module.Write Scale=false
    Module.Write SpriteRotation=true
    Module.Write SpriteSize=true
    Position=<0 0 75>
    Sprite Rotation=&NormalizedAngleToDegrees(NormalizedAngle=&RandomRangeFloat(Minimum=0, Maximum=1, Evaluation Type=::NewEnumerator1))
    Sprite Size=&Vector2DFromFloat(Value=&RandomRangeFloat(Minimum=80, Maximum=90, Evaluation Type=::NewEnumerator1))
   )
   &SkeletalMeshLocation(
    Apply To Particle Position=false
    Inherit Velocity=false
    MaskSpawnGroup=false
    Mesh Sampling Type=::NewEnumerator5
    Skeletal Mesh=$User.SkeletalMesh
    Socket Sampling Mode=::NewEnumerator5
   ) {module=SkeletalMeshLocation001}
   &SphereLocation(
    MaskSpawnGroup=false
    Offset=<0 -70 25>
    Override Seed=false
    Sphere Origin:Vector3f=$Output.SkeletalMeshLocation001.SampledPosition
    Sphere Radius=40
    Surface Only=false
    UseNonUniformScale=false
    UseOffset=true
    UseSphereOrientation=false
   )
   &CameraOffset(Camera Offset Amount=&RandomRangeFloat(Minimum=2, Maximum=5))
   &SubUVAnimation(End Frame=3, SubUV Animation Mode:ENiagara_SubUVLookupMode=::NewEnumerator2)
   &AddVelocityFromPoint(Module.Use Falloff Distance=false, Velocity Strength=&RandomRangeFloat(Minimum=250, Maximum=300)) {enabled=false, module=AddVelocityFromPoint001}
   &AddVelocity(Velocity=&RandomRangeVector(Minimum=<0 0 80>, Maximum=<0 0 100>))
   &AddVelocity(Velocity=<0 -200 0>) {module=AddVelocity001}
  ]
  particleUpdateScript [
   &ParticleState()
   &CurlNoiseForce(Noise Frequency=51.8223, Noise Strength=12.4513, Pan Noise=false)
   &GravityForce(Gravity=<0 0 -120>)
   &ScaleSpriteSize(Scale Factor=&Vector2DFromFloat(Value=&FloatFromCurve(FloatCurve=DataInterface() {propertyValues {Curve {keys [
          {interpMode=RCIM_Cubic, tangentMode=RCTM_User, time=0, value=0}
          {arriveTangent=0.7232, interpMode=RCIM_Cubic, leaveTangent=0.7232, tangentMode=RCTM_User, time=0.1, value=0.8}
          {arriveTangent=-0.0021, interpMode=RCIM_Cubic, leaveTangent=-0.0021, tangentMode=RCTM_User, time=0.8, value=1.2}
          {tangentMode=RCTM_User, time=1, value=0}
         ]}}})))
   &Drag(Use Linear Drag=true, Use Rotational Drag=false)
   &DynamicMaterialParameters(
    Index 0 Param 1=&FloatFromCurve(FloatCurve=DataInterface() {propertyValues {Curve {keys [
         {interpMode=RCIM_Cubic, leaveTangentWeight=0.0333, tangentMode=RCTM_User, time=0.1, value=1.25}
         {interpMode=RCIM_Cubic, tangentMode=RCTM_User, time=0.2, value=2}
         {arriveTangent=-1.7657, interpMode=RCIM_Cubic, leaveTangent=-1.7657, tangentMode=RCTM_User, time=0.4, value=0.75}
         {arriveTangentWeight=0.2333, interpMode=RCIM_Cubic, tangentMode=RCTM_User, time=1.1, value=0.1}
        ]}}})
    Index 0 Param 3=1
    Index 0 Param 4=1
    Index 1 Param 1=1
    Index 1 Param 2=1
    Index 2 Param 4=0.2
    Index 3 Param 1=1
    Index 3 Param 2=1
    Module.Param0WriteEnabled=true
    Module.Param0WriteEnabled1=true
    Module.Param0WriteEnabled2=true
    Module.Param0WriteEnabled3=true
    Module.Param1WriteEnabled=true
    Module.Param1WriteEnabled1=true
    Module.Param1WriteEnabled2=true
    Module.Param1WriteEnabled3=true
    Module.Param2WriteEnabled=true
    Module.Param2WriteEnabled1=true
    Module.Param2WriteEnabled2=true
    Module.Param2WriteEnabled3=false
    Module.Param3WriteEnabled=true
    Module.Param3WriteEnabled1=true
    Module.Param3WriteEnabled2=true
    Module.Param3WriteEnabled3=false
    Write Parameter Index 1=true
    Write Parameter Index 2=true
    Write Parameter Index 3=true
   )
   &SpriteRotationRate(Rotation Rate=&NormalizedAngleToDegrees(NormalizedAngle=$Particles.SpriteRotationRate))
   &SolveForcesAndVelocity(Clamp Velocity=false, Limit Acceleration=false)
   &ScaleColor(
    Scale Alpha=&FloatFromCurve(
     FloatCurve=DataInterface() {propertyValues {Curve {keys [
         {interpMode=RCIM_Cubic, tangentMode=RCTM_User, time=0, value=0}
         {interpMode=RCIM_Cubic, tangentMode=RCTM_User, time=0.1, value=1}
         {tangentMode=RCTM_User, time=1, value=0}
        ]}}}
     Scale Curve=2
    )
    ScaleA=true
    ScaleRGB=true
   ) {module=ScaleColor001}
  ]
  rendererClasses [&NiagaraSpriteRendererProperties]
  renderers [{rendererClass=&NiagaraSpriteRendererProperties}]
 }
 C_Splash {
  emitterUpdateScript [
   &EmitterState()
   &SpawnRate(Spawn Probability=0.2, SpawnRate=60, Use Spawn Probability=true) {enabled=false}
   &SpawnBurst_Instantaneous(Spawn Count=&RandomRangeInt(Minimum=5, Maximum=7), Use Spawn Probability=false)
  ]
  particleSpawnScript [
   /Niagara/Modules/Spawn/Initialization/V2/InitializeParticle(
    Color=<rgba 0 0 0 4>
    Lifetime=&RandomRangeFloat(Minimum=0.2, Maximum=0.6)
    Sprite Rotation Mode=::NewEnumerator2
    Sprite Size=&Multiply_Vector2(A=&RandomRangeVector2D(Minimum=<75 75>, Maximum=<90 90>), B=<0.8 0.8>)
    Sprite Size Mode=::NewEnumerator4
    UsePositionOffset=false
    Write Material Random=true
   )
   &AddVelocity(Velocity=&Multiply_VectorByFloat(Vector=&RandomRangeVector(Minimum=<-1 -1 -1>, Maximum=<1 1 1.5>), Float=100)) {enabled=false}
   &SkeletalMeshLocation(
    Apply To Particle Position=false
    Inherit Velocity=false
    MaskSpawnGroup=false
    Mesh Sampling Type=::NewEnumerator5
    Skeletal Mesh=$User.SkeletalMesh
    Socket Sampling Mode=::NewEnumerator5
   ) {module=SkeletalMeshLocation001}
   &SphereLocation(
    MaskSpawnGroup=false
    Offset=<0 -70 25>
    Override Seed=false
    Sphere Origin:Vector3f=$Output.SkeletalMeshLocation001.SampledPosition
    Sphere Radius=80
    Surface Only=false
    UseNonUniformScale=false
    UseOffset=true
    UseSphereOrientation=false
   )
   &CylinderLocation(
    Cylinder Height=0
    Cylinder Radius=30
    MaskSpawnGroup=false
    Module.Enable Lathe Profile=false
    Module.Override Randomness=false
    Module.Surface Only=true
    Orientation Axis:ENiagaraOrientationAxis=::XAxis
    Surface Only Band Thickness=3
   ) {enabled=false}
   &AddVelocityFromPoint(Module.Use Falloff Distance=false, Velocity Strength=&RandomRangeFloat(Minimum=250, Maximum=300)) {enabled=false}
   &CameraOffset(Camera Offset Amount=&RandomRangeFloat(Minimum=0, Maximum=10))
   &AddVelocity(Velocity=<0 -200 0>) {module=AddVelocity001}
  ]
  particleUpdateScript [
   &ParticleState()
   &CurlNoiseForce(Noise Frequency=51.8223, Noise Strength=12.4513, Pan Noise=false)
   &GravityForce(Gravity=<0 0 -300>)
   &SolveForcesAndVelocity(Clamp Velocity=false, Limit Acceleration=false)
   &ScaleColor(
    Scale Alpha=&FloatFromCurve(
     FloatCurve=DataInterface() {propertyValues {Curve {keys [<at 0 1 RCIM_Cubic>, {arriveTangent=-3.2218, interpMode=RCIM_Cubic, leaveTangent=-3.2218, tangentMode=RCTM_User, time=1, value=0}]}}}
     Scale Curve=1.5
    )
    ScaleA=true
    ScaleRGB=true
   )
   &DynamicMaterialParameters(
    Index 0 Param 1=&FloatFromCurve(
     FloatCurve=DataInterface() {propertyValues {Curve {keys [<at 0 0 RCIM_Cubic>, {arriveTangent=3.8258, interpMode=RCIM_Cubic, leaveTangent=3.8258, tangentMode=RCTM_User, time=1, value=4}]}}}
     Scale Curve=&RandomRangeFloat(Minimum=0.8, Maximum=1.5)
    )
    Index 0 Param 2=&FloatFromCurve(
     FloatCurve=DataInterface() {propertyValues {Curve {keys [<at 0 0 RCIM_Cubic>, {arriveTangent=2.9697, interpMode=RCIM_Cubic, leaveTangent=2.9697, tangentMode=RCTM_User, time=1, value=2}]}}}
     Scale Curve=&RandomRangeFloat(Minimum=0.8, Maximum=1.5)
    )
    Index 0 Param 3=&RandomRangeFloat(Minimum=0.5, Maximum=1.5)
    Index 0 Param 4=&RandomRangeFloat(Minimum=0.5, Maximum=1.5)
    Module.Param0WriteEnabled=true
    Module.Param1WriteEnabled=true
    Module.Param2WriteEnabled=true
    Module.Param3WriteEnabled=true
   )
   &ScaleSpriteSize(Scale Factor=&Vector2DFromFloat(Value=&FloatFromCurve(FloatCurve=DataInterface() {propertyValues {Curve {keys [<at 0 0.3>,<at 0.5 1>]}}}))) {module=ScaleSpriteSize001}
  ]
  rendererClasses [&NiagaraSpriteRendererProperties]
  renderers [{rendererClass=&NiagaraSpriteRendererProperties}]
 }
 Drops {
  emitterUpdateScript [
   &EmitterState()
   &SpawnRate(Spawn Probability=0.2, SpawnRate=60, Use Spawn Probability=true) {enabled=false}
   &SpawnBurst_Instantaneous(Spawn Count=&RandomRangeInt(Minimum=5, Maximum=8), Use Spawn Probability=false)
  ]
  particleSpawnScript [
   &InitializeParticle(
    Lifetime=&RandomRangeFloat(Minimum=0.6, Maximum=0.8)
    Sprite Size=&Multiply_Vector2(A=&RandomRangeVector2D(Minimum=<2 4>, Maximum=<4 6>), B=<1.5 1.5>)
    Sprite Size Mode=::NewEnumerator4
    UsePositionOffset=false
    Write Material Random=true
   )
   &AddVelocity(Velocity=&Multiply_VectorByFloat(Vector=&RandomRangeVector(Minimum=<-1 -1 -1>, Maximum=<1 1 1.5>), Float=100))
   &SkeletalMeshLocation(
    Apply To Particle Position=false
    Inherit Velocity=false
    MaskSpawnGroup=false
    Mesh Sampling Type=::NewEnumerator5
    Skeletal Mesh=$User.SkeletalMesh
    Socket Sampling Mode=::NewEnumerator5
   ) {module=SkeletalMeshLocation001}
   &SphereLocation(
    MaskSpawnGroup=false
    Offset=<0 -60 25>
    Override Seed=false
    Sphere Origin:Vector3f=$Output.SkeletalMeshLocation001.SampledPosition
    Sphere Radius=80
    Surface Only=false
    UseNonUniformScale=false
    UseOffset=true
    UseSphereOrientation=false
   )
   &AddVelocityFromPoint(Module.Use Falloff Distance=false, Velocity Strength=&RandomRangeFloat(Minimum=180, Maximum=220)) {enabled=false}
   &CameraOffset(Camera Offset Amount=&RandomRangeFloat(Minimum=10, Maximum=20))
   &AddVelocity(Velocity=<0 -200 0>) {module=AddVelocity001}
  ]
  particleUpdateScript [
   &CurlNoiseForce(Noise Frequency=51.8223, Noise Strength=12.4513, Pan Noise=false)
   &GravityForce(Gravity=<0 0 -150>)
   &SolveForcesAndVelocity(Clamp Velocity=false, Limit Acceleration=false)
   &ParticleState()
   &ScaleColor(
    Scale Alpha=&FloatFromCurve(FloatCurve=DataInterface() {propertyValues {Curve {keys [<at 0 1 RCIM_Cubic>, {arriveTangent=-3.2218, interpMode=RCIM_Cubic, leaveTangent=-3.2218, tangentMode=RCTM_User, time=1, value=0}]}}})
    ScaleA=true
    ScaleRGB=true
   )
   &ScaleSpriteSize(Scale Factor=&Vector2DFromFloat(Value=&FloatFromCurve(FloatCurve=DataInterface() {propertyValues {Curve {keys [<at 0 1 RCIM_Cubic>, {arriveTangent=-3.6731, interpMode=RCIM_Cubic, leaveTangent=-3.6731, tangentMode=RCTM_User, time=1, value=0}]}}})))
  ]
  rendererClasses [&NiagaraSpriteRendererProperties]
  renderers [{rendererClass=&NiagaraSpriteRendererProperties}]
 }
 Slurpy_Streaks {
  emitterUpdateScript [
   &EmitterState()
   &SpawnBurst_Instantaneous(Spawn Count=&RandomRangeInt(Minimum=2, Maximum=3), Use Spawn Probability=false)
  ]
  particleSpawnScript [
   /SetVariables(Particles.SpriteRotationRate=&RandomRangeFloat(Minimum=-0.1, Maximum=0.1, Evaluation Type=::NewEnumerator1)) {isSetParameters=true, module=SetVariables_#1}
   /Niagara/Modules/Spawn/Initialization/InitializeParticle(
    Lifetime=&RandomRangeFloat(Minimum=0.3, Maximum=0.6, Evaluation Type=::NewEnumerator1)
    Module.Write Color=true
    Module.Write Lifetime=true
    Module.Write Mass=true
    Module.Write Material Random=true
    Module.Write Position=false
    Module.Write Scale=false
    Module.Write SpriteRotation=true
    Module.Write SpriteSize=true
    Position=<0 0 75>
    Sprite Rotation=&NormalizedAngleToDegrees(NormalizedAngle=&RandomRangeFloat(Minimum=0, Maximum=1, Evaluation Type=::NewEnumerator1))
    Sprite Size=&Vector2DFromFloat(Value=&RandomRangeFloat(Minimum=70, Maximum=80, Evaluation Type=::NewEnumerator1))
   )
   &SkeletalMeshLocation(
    Apply To Particle Position=false
    Inherit Velocity=false
    MaskSpawnGroup=false
    Mesh Sampling Type=::NewEnumerator5
    Skeletal Mesh=$User.SkeletalMesh
    Socket Sampling Mode=::NewEnumerator5
   ) {module=SkeletalMeshLocation001}
   &SphereLocation(
    MaskSpawnGroup=false
    Offset=<0 -80 25>
    Override Seed=false
    Sphere Origin:Vector3f=$Output.SkeletalMeshLocation001.SampledPosition
    Sphere Radius=60
    Surface Only=false
    UseNonUniformScale=false
    UseOffset=true
    UseSphereOrientation=false
   )
   &CameraOffset(Camera Offset Amount=&RandomRangeFloat(Minimum=0, Maximum=10))
   &SubUVAnimation(End Frame=3, SubUV Animation Mode:ENiagara_SubUVLookupMode=::NewEnumerator2)
   &AddVelocityFromPoint(Module.Use Falloff Distance=false, Velocity Strength=&RandomRangeFloat(Minimum=350, Maximum=500)) {enabled=false, module=AddVelocityFromPoint001}
   &AddVelocity(Velocity=<0 -200 0>) {module=AddVelocity001}
  ]
  particleUpdateScript [
   &ParticleState()
   &CurlNoiseForce(Noise Frequency=51.8223, Noise Strength=12.4513, Pan Noise=false)
   &GravityForce(Gravity=<0 0 -120>)
   &ScaleSpriteSize(Scale Factor=&Vector2DFromFloat(Value=&FloatFromCurve(FloatCurve=DataInterface() {propertyValues {Curve {keys [
          {interpMode=RCIM_Cubic, tangentMode=RCTM_User, time=0, value=0}
          {arriveTangent=0.7232, interpMode=RCIM_Cubic, leaveTangent=0.7232, tangentMode=RCTM_User, time=0.1, value=0.8}
          {arriveTangent=-0.0021, interpMode=RCIM_Cubic, leaveTangent=-0.0021, tangentMode=RCTM_User, time=0.8, value=1.2}
          {tangentMode=RCTM_User, time=1, value=0}
         ]}}})))
   &Drag(Use Linear Drag=true, Use Rotational Drag=false)
   &DynamicMaterialParameters(
    Index 0 Param 1=&FloatFromCurve(FloatCurve=DataInterface() {propertyValues {Curve {keys [
         {interpMode=RCIM_Cubic, leaveTangentWeight=0.0333, tangentMode=RCTM_User, time=0.1, value=1.25}
         {interpMode=RCIM_Cubic, tangentMode=RCTM_User, time=0.2, value=2}
         {arriveTangent=-1.7657, interpMode=RCIM_Cubic, leaveTangent=-1.7657, tangentMode=RCTM_User, time=0.4, value=0.75}
         {arriveTangentWeight=0.2333, interpMode=RCIM_Cubic, tangentMode=RCTM_User, time=1.1, value=0.1}
        ]}}})
    Index 0 Param 3=1
    Index 0 Param 4=1
    Module.Param0WriteEnabled=true
    Module.Param1WriteEnabled=true
    Module.Param2WriteEnabled=true
    Module.Param3WriteEnabled=true
   )
   &SpriteRotationRate(Rotation Rate=&NormalizedAngleToDegrees(NormalizedAngle=$Particles.SpriteRotationRate))
   &SolveForcesAndVelocity(Clamp Velocity=false, Limit Acceleration=false)
   &ScaleColor(
    Scale Alpha=&FloatFromCurve(
     FloatCurve=DataInterface() {propertyValues {Curve {keys [
         {interpMode=RCIM_Cubic, tangentMode=RCTM_User, time=0, value=0}
         {interpMode=RCIM_Cubic, tangentMode=RCTM_User, time=0.1, value=1}
         {tangentMode=RCTM_User, time=1, value=0}
        ]}}}
     Scale Curve=2
    )
    ScaleA=true
    ScaleRGB=true
   ) {module=ScaleColor001}
  ]
  rendererClasses [&NiagaraSpriteRendererProperties]
  renderers [{rendererClass=&NiagaraSpriteRendererProperties}]
 }
}
summary {
 systemName=NS_Glider_Accumulate_Deploy_02_FE
 emitters [
  {emitterName=Drops, rendererClasses [&NiagaraSpriteRendererProperties]}
  {emitterName=C_Splash, rendererClasses [&NiagaraSpriteRendererProperties]}
  {emitterName=Slurpy_Streaks, rendererClasses [&NiagaraSpriteRendererProperties]}
  {emitterName=BigSplat, rendererClasses [&NiagaraSpriteRendererProperties]}
 ]
 userVariables [User.SkeletalMesh {
   type=NiagaraDataInterfaceSkeletalMesh
   defaultValue {
    struct=NiagaraExt_VariableValue_DataInterface
    value {dataInterface=&self:NiagaraDataInterfaceSkeletalMesh_0}
   }
  }]
}
