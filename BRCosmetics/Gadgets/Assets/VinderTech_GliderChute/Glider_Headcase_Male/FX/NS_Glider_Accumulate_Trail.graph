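# Niagara trail effect for the VinderTech Headcase Male glider, shedding mesh drips, drool strands and a cone.
# Drips and LargeDrips spawn around the glider skeletal mesh; their rates, like DroolPendulum's, scale with a system integrator value in gameplay, with front-end overrides.
# DroolPendulum simulates pendulums pivoting on SKM_Back, stretching meshes to the pivot and passing endpoints and owner velocity direction to the material.
# OuterCone follows the mesh, spawning by User.ForwardDot.
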NiagaraSystem /BRCosmetics/Gadgets/Assets/VinderTech_GliderChute/Glider_Headcase_Male/FX/NS_Glider_Accumulate_Trail
dependencies {
 usedDataInterfaces [&NiagaraDataInterfaceCurve]
 usedDynamicInputs []
 usedModules [&SystemState, &FN_Integrator]
}
emitters {
 Drips {
  emitterUpdateScript [
   &EmitterState()
   &SpawnRate(
    SpawnRate=&FN_FrontEndFloatModifier(Gameplay=&Lerp_Float(A=0, B=5, Alpha=$System.FN_Integrator002.Value), Front End=5)
    Use Spawn Probability=false
   )
  ]
  particleSpawnScript [
   /Niagara/Modules/Spawn/Initialization/InitializeParticle(
    Lifetime=&FN_FrontEndFloatModifier(
     Gameplay=&UniformAOrBFloat(A=0.8, B=1.2)
     Front End=&RandomRangeFloat(Minimum=0.25, Maximum=0.5)
    )
    Mesh Scale=&Lerp_Vector(
     A=<4 4 4>
     B=<6 6 6>
     Alpha=&FloatFromCurve(
      FloatCurve=DataInterface() {propertyValues {Curve {keys [<at 0.1 0>,<at 0.5 0.1>,<at 0.8 0.2>,<at 0.9 0.4>,<at 1 1>]}}}
      CurveIndex=&RandomRangeFloat(Minimum=0, Maximum=1)
     )
    )
    Module.Write Color=true
    Module.Write Lifetime=true
    Module.Write Mass=true
    Module.Write Material Random=true
    Module.Write Position=false
    Module.Write Scale=true
    Module.Write SpriteRotation=false
    Module.Write SpriteSize=true
    Sprite Size=&Vector2DFromFloat(Value=5)
   )
   &FN_SetSpawnGroup()
   &SkeletalMeshLocation(
    Inherit Velocity=false
    MaskSpawnGroup=false
    Mesh Sampling Type=::NewEnumerator5
    Skeletal Mesh=$User.SkeletalMesh
    Socket Sampling Mode=::NewEnumerator5
   )
   &CylinderLocation(
    Cylinder Height=5
    Cylinder Origin:Vector3f=$Output.SkeletalMeshLocation.SampledPosition
    Cylinder Radius=65
    MaskSpawnGroup=false
    Module.Enable Lathe Profile=false
    Module.Override Randomness=false
    Module.Surface Only=true
    Non Uniform Scale:Vector3f=<0.85 1 1>
    Orientation Axis:ENiagaraOrientationAxis=::YAxis
    Use Endcaps In Surface Only Mode=false
   )
   &AddVelocity(Velocity=&FN_FrontEndVector3Modifier(Gameplay=<0 0 0>, Front End=<0 -200 0>))
   /SetVariables(Particles.VisibilityTag=&MaskIntBySpawnGroup(Masked Int=0, Passthrough Int=1, Spawn Group Mask=0)) {isSetParameters=true}
  ]
  particleUpdateScript [
   &ParticleState()
   &DynamicMaterialParameters(
    Index 0 Param 1=&RandomRangeFloat(Minimum=0, Maximum=1)
    Index 0 Param 2=1
    Module.Param0WriteEnabled=true
    Module.Param1WriteEnabled=true
    Module.Param2WriteEnabled=true
    Module.Param3WriteEnabled=true
   )
   &ScaleMeshSize(Scale Factor=&Multiply_VectorByFloat(
     Vector=&MakeVector(
      X=&Lerp_Float(
       A=0.25
       B=1
       Alpha=&FloatFromCurve(FloatCurve=DataInterface() {propertyValues {Curve {keys [
            {arriveTangent=4.1785, interpMode=RCIM_Cubic, leaveTangent=4.1785, tangentMode=RCTM_User, time=0, value=0}
            {arriveTangent=0.9995, interpMode=RCIM_Cubic, leaveTangent=0.9995, tangentMode=RCTM_User, time=0.2223, value=0.6825}
            {arriveTangent=-0.0405, interpMode=RCIM_Cubic, leaveTangent=-0.0405, tangentMode=RCTM_User, time=1, value=1}
           ]}}})
      )
      Y=&Lerp_Float(
       A=1
       B=0.25
       Alpha=&FloatFromCurve(FloatCurve=DataInterface() {propertyValues {Curve {keys [
            {arriveTangent=4.1785, interpMode=RCIM_Cubic, leaveTangent=4.1785, tangentMode=RCTM_User, time=0, value=0}
            {arriveTangent=0.9995, interpMode=RCIM_Cubic, leaveTangent=0.9995, tangentMode=RCTM_User, time=0.2223, value=0.6825}
            {arriveTangent=-0.0405, interpMode=RCIM_Cubic, leaveTangent=-0.0405, tangentMode=RCTM_User, time=1, value=1}
           ]}}})
      )
      Z=&Lerp_Float(
       A=1
       B=0.25
       Alpha=&FloatFromCurve(FloatCurve=DataInterface() {propertyValues {Curve {keys [
            {arriveTangent=4.1785, interpMode=RCIM_Cubic, leaveTangent=4.1785, tangentMode=RCTM_User, time=0, value=0}
            {arriveTangent=0.9995, interpMode=RCIM_Cubic, leaveTangent=0.9995, tangentMode=RCTM_User, time=0.2223, value=0.6825}
            {arriveTangent=-0.0405, interpMode=RCIM_Cubic, leaveTangent=-0.0405, tangentMode=RCTM_User, time=1, value=1}
           ]}}})
      )
     )
     Float=&Lerp_Float(
      A=0
      B=1
      Alpha=&FloatFromCurve(FloatCurve=DataInterface() {propertyValues {Curve {keys [
           {arriveTangent=9.0725, interpMode=RCIM_Cubic, leaveTangent=9.0725, tangentMode=RCTM_User, time=0, value=0}
           {arriveTangent=0.7857, interpMode=RCIM_Cubic, leaveTangent=0.7857, time=0.1408, value=0.9847}
           {arriveTangent=-2.2398, interpMode=RCIM_Cubic, leaveTangent=-2.2398, tangentMode=RCTM_User, time=0.7, value=0.55}
           {arriveTangent=-0.0405, interpMode=RCIM_Cubic, leaveTangent=-0.0405, tangentMode=RCTM_User, time=1, value=0}
          ]}}})
     )
    ))
   &InheritSourceMovement(
    Applied Force Scale=<0.1 0.1 0.1>
    Applied Position Falloff Distance=100
    Applied Position Scale=<0.8 0.8 0.8>
    ApplyForce=true
    ApplyForceFalloff=false
    ApplyPosition=true
    ApplyPositionFalloff=false
    Module.LimitForceSpeed=false
   )
   &SolveForcesAndVelocity(Clamp Velocity=false, Limit Acceleration=false)
   &ScaleMeshSize(Scale Factor=&VectorFromFloat(Value=&MaskFloatBySpawnGroup(Masked Float=0.7, Passthrough Float=1))) {module=ScaleMeshSize001}
  ]
  rendererClasses [&NiagaraMeshRendererProperties]
  renderers [
   {rendererClass=&NiagaraMeshRendererProperties, rendererIndex=0}
   {rendererClass=&NiagaraMeshRendererProperties, rendererIndex=1}
  ]
 }
 DroolPendulum {
  emitterUpdateScript [
   &EmitterState()
   &SpawnRate(
    SpawnRate=&FN_FrontEndFloatModifier(Gameplay=&Lerp_Float(A=0, B=1, Alpha=$System.FN_Integrator002.Value), Front End=0)
    Use Spawn Probability=false
   )
  ]
  particleSpawnScript [
   /SetVariables(
    Particles.PickAll:NiagaraInt32=&RandomRangeInt(Minimum=0, Maximum=1810)
    Particles.PickBack:NiagaraInt32=&RandomRangeInt(Minimum=1252, Maximum=1171)
   ) {isSetParameters=true}
   /SetVariables(
    Particles.PendulumLength:NiagaraFloat=&RandomRangeFloat(Minimum=6, Maximum=8)
    Particles.PickBone:NiagaraInt32=&MakeCustomIntFromBool(
     Bool=&SetBoolByIntComparison(A=&RandomRangeInt(Minimum=0, Maximum=1), B=0, Comparison Type=::NewEnumerator4)
     True INT=$Particles.PickBack
     False INT=$Particles.PickAll
    )
    Particles.ScaleLerp:NiagaraFloat=&RandomRangeFloat(Minimum=0, Maximum=1)
   ) {isSetParameters=true}
   &SkeletalMeshLocation(
    Bone / Socket Index=&RandomRangeInt(Minimum=0, Maximum=2)
    Inherit Velocity=false
    MaskSpawnGroup=false
    Mesh Sampling Type=::NewEnumerator5
    Skeletal Mesh=$User.SKM_Back
    Socket Sampling Mode=::NewEnumerator5
   ) {module=SkeletalMeshLocation003}
   &SphereLocation(
    MaskSpawnGroup=false
    Override Seed=false
    Sphere Radius=200
    Surface Only=false
    UseNonUniformScale=false
    UseOffset=false
    UseSphereOrientation=false
    Write to Intrinsic Variables=false
   )
   &InitializeParticle(
    Lifetime=&Power(a=&UniformRangedFloat(Minimum=0.8, Maximum=1), b=2)
    Mesh Scale=&Lerp_Vector(A=<0.3 0.3 0.02>, B=<0.4 0.4 0.02>, Alpha=$Particles.ScaleLerp)
    Module.Write Color=true
    Module.Write Lifetime=true
    Module.Write Mass=true
    Module.Write Material Random=true
    Module.Write Position=false
    Module.Write Scale=true
    Module.Write SpriteRotation=false
    Module.Write SpriteSize=true
    Sprite Size=&Vector2DFromFloat(Value=5)
   )
  ]
  particleUpdateScript [
   &ParticleState()
   &SkeletalMeshLocation(
    Apply To Particle Position=false
    Bone / Socket Index=&RandomRangeInt(Minimum=0, Maximum=2)
    Inherit Velocity=false
    MaskSpawnGroup=false
    Mesh Sampling Type=::NewEnumerator5
    Skeletal Mesh=$User.SKM_Back
    Socket Sampling Mode=::NewEnumerator5
   )
   /SetVariables(Particles.MeshPosition:Vector3f=$Output.SkeletalMeshLocation.SampledPosition) {isSetParameters=true}
   /SetVariables(
    Particles.AlphaScale:NiagaraFloat=&FloatFromCurve(FloatCurve=DataInterface() {propertyValues {Curve {keys [
         {arriveTangent=-0.0275, interpMode=RCIM_Cubic, leaveTangent=-0.0275, tangentMode=RCTM_User, time=0.95, value=1}
         {arriveTangent=-34.2468, interpMode=RCIM_Cubic, leaveTangent=-34.2466, tangentMode=RCTM_User, time=1, value=0}
        ]}}})
    Particles.XYScale:NiagaraFloat=&FloatFromCurve(FloatCurve=DataInterface() {propertyValues {Curve {keys [
         {arriveTangent=10.1742, interpMode=RCIM_Cubic, leaveTangent=10.1742, tangentMode=RCTM_User, time=0, value=0}
         {arriveTangent=-2.2106, interpMode=RCIM_Cubic, leaveTangent=-2.2106, tangentMode=RCTM_User, time=0.1, value=1}
         {arriveTangent=-0.0096, interpMode=RCIM_Cubic, leaveTangent=-0.0096, tangentMode=RCTM_User, time=0.4, value=0.5}
         {arriveTangent=-3.238, interpMode=RCIM_Cubic, leaveTangent=-3.238, tangentMode=RCTM_User, time=1, value=0.1}
        ]}}})
    Particles.ZScale:NiagaraFloat=&FloatFromCurve(
     FloatCurve=DataInterface() {propertyValues {Curve {keys [
         {arriveTangent=4.2799, interpMode=RCIM_Cubic, leaveTangent=4.2799, tangentMode=RCTM_User, time=0, value=0.1}
         {arriveTangent=1.4091, interpMode=RCIM_Cubic, leaveTangent=1.4091, tangentMode=RCTM_User, time=0.5, value=1}
         {arriveTangent=10.2989, interpMode=RCIM_Cubic, leaveTangent=10.2989, tangentMode=RCTM_User, time=1, value=3}
        ]}}}
     Scale Curve=$Particles.PendulumLength
    )
   ) {isSetParameters=true}
   &PendulumSetup(
    Module.Enable Spring Driver=false
    Pendulum Length:NiagaraFloat=$Particles.ZScale
    Pendulum Pivot:Vector3f=$Particles.MeshPosition
    Rigid Pendulum=false
   )
   &Drag(Drag=2, Use Linear Drag=true, Use Rotational Drag=false)
   &AddVelocity(Velocity=&FN_FrontEndVector3Modifier(
     Gameplay=<0 0 0>
     Front End=&RandomRangeVector(Minimum=<-50 0 -20>, Maximum=<-200 100 100>, Evaluation Type=::NewEnumerator1)
    ))
   &SolveForcesAndVelocity(Clamp Velocity=false, Limit Acceleration=false)
   &PendulumConstraint()
   &CalculateAccurateVelocity()
   /SetVariables(Particles.MeshFacing:Vector3f=&NormalizeVector(Vector To Normalize=&Subtract_Vector(A=$Particles.MeshPosition, B=Linked() {linkedVariable {name=Particles.Position, type=NiagaraPosition}}))) {isSetParameters=true}
   /Niagara/Modules/Update/Orientation/V2/OrientMeshToVector(
    Look At Direction:Vector3f=&Multiply_VectorByFloat(Vector=$Particles.MeshFacing, Float=1)
    Override Quat=false
    Reference Axis Vector:Vector3f=<0 0 1>
    UseOptionalSideAxis=false
    UseOptionalUpAxis=false
   )
   &SkeletalMeshLocation(
    Apply To Particle Position=false
    Bone / Socket Index=&RandomRangeInt(Minimum=0, Maximum=2)
    Inherit Velocity=false
    MaskSpawnGroup=false
    Mesh Sampling Type=::NewEnumerator5
    Skeletal Mesh=$User.SKM_Back
    Socket Sampling Mode=::NewEnumerator5
   ) {module=SkeletalMeshLocation001}
   /SetVariables(Particles.RestDirection:Vector3f=/Niagara/DynamicInputs/Helpers/SafeNormalizeVector(Vector To Normalize:Vector3f=$Engine.Owner.Velocity, Fallback Vector:Vector3f=<0 0 0>)) {isSetParameters=true}
   &DynamicMaterialParameters(
    Index 1 Param 1=HlslExpression() {hlslExpression=Particles.RestDirection.x}
    Index 1 Param 2=HlslExpression() {hlslExpression=Particles.RestDirection.y}
    Index 1 Param 3=HlslExpression() {hlslExpression=Particles.RestDirection.z}
    Index 2 Param 1=HlslExpression() {hlslExpression=Particles.MeshPosition.x}
    Index 2 Param 2=HlslExpression() {hlslExpression=Particles.MeshPosition.y}
    Index 2 Param 3=HlslExpression() {hlslExpression=Particles.MeshPosition.z}
    Index 2 Param 4=0.75
    Index 3 Param 1=&MakeFloatFromVector(Vector=Linked() {linkedVariable {name=Particles.Position, type=NiagaraPosition}}, Channel=::NewEnumerator0)
    Index 3 Param 2=&MakeFloatFromVector(Vector=Linked() {linkedVariable {name=Particles.Position, type=NiagaraPosition}}, Channel=::NewEnumerator1)
    Index 3 Param 3=&MakeFloatFromVector(Vector=Linked() {linkedVariable {name=Particles.Position, type=NiagaraPosition}}, Channel=::NewEnumerator2)
    Index 3 Param 4=0.3
    Module.Param0WriteEnabled1=true
    Module.Param0WriteEnabled2=true
    Module.Param0WriteEnabled3=true
    Module.Param1WriteEnabled1=true
    Module.Param1WriteEnabled2=true
    Module.Param1WriteEnabled3=true
    Module.Param2WriteEnabled1=true
    Module.Param2WriteEnabled2=true
    Module.Param2WriteEnabled3=true
    Module.Param3WriteEnabled1=false
    Module.Param3WriteEnabled2=true
    Module.Param3WriteEnabled3=true
    Write Parameter Index 0=false
    Write Parameter Index 1=true
    Write Parameter Index 2=true
    Write Parameter Index 3=true
   )
   &ScaleMeshSize(Scale Factor=&MakeVector(
     X=$Particles.XYScale
     Y=$Particles.XYScale
     Z=&VectorLength(Vector=&Subtract_Vector(A=$Particles.MeshPosition, B=Linked() {linkedVariable {name=Particles.Position, type=NiagaraPosition}}))
    ))
   &ScaleColor(Scale Alpha=$Particles.AlphaScale, ScaleA=true, ScaleRGB=false)
  ]
  rendererClasses [&NiagaraMeshRendererProperties]
  renderers [{rendererClass=&NiagaraMeshRendererProperties, rendererIndex=0}]
 }
 LargeDrips {
  emitterUpdateScript [
   &EmitterState()
   &SpawnRate(
    SpawnRate=&FN_FrontEndFloatModifier(Gameplay=&Lerp_Float(A=0, B=2, Alpha=$System.FN_Integrator002.Value), Front End=2)
    Use Spawn Probability=false
   )
  ]
  particleSpawnScript [
   &InitializeParticle(
    Lifetime=&RandomRangeFloat(Minimum=1.2, Maximum=1.4)
    Mesh Scale=&Lerp_Vector(
     A=<8 8 8>
     B=<14 14 14>
     Alpha=&FloatFromCurve(
      FloatCurve=DataInterface() {propertyValues {Curve {keys [<at 0.1 0>,<at 0.5 0.1>,<at 0.8 0.2>,<at 0.9 0.4>,<at 1 1>]}}}
      CurveIndex=&RandomRangeFloat(Minimum=0, Maximum=1)
     )
    )
    Module.Write Color=true
    Module.Write Lifetime=true
    Module.Write Mass=true
    Module.Write Material Random=true
    Module.Write Position=false
    Module.Write Scale=true
    Module.Write SpriteRotation=false
    Module.Write SpriteSize=true
    Sprite Size=&Vector2DFromFloat(Value=5)
   )
   &SkeletalMeshLocation(
    Inherit Velocity=false
    MaskSpawnGroup=false
    Mesh Sampling Type=::NewEnumerator5
    Skeletal Mesh=$User.SkeletalMesh
    Socket Sampling Mode=::NewEnumerator5
   )
   &CylinderLocation(
    Cylinder Height=5
    Cylinder Origin:Vector3f=$Output.SkeletalMeshLocation.SampledPosition
    Cylinder Radius=50
    Lathe Profile=DataInterface() {propertyValues {Curve {keys [<at 0 0.451 RCIM_Cubic>, {arriveTangent=0.049, interpMode=RCIM_Cubic, leaveTangent=0.049, time=0.5, value=1}, <at 1 0.5 RCIM_Cubic>]}}}
    MaskSpawnGroup=false
    Module.Enable Lathe Profile=false
    Module.Override Randomness=false
    Module.Surface Only=true
    Orientation Axis:ENiagaraOrientationAxis=::YAxis
    Use Endcaps In Surface Only Mode=false
   )
   &AddVelocity(Velocity=&FN_FrontEndVector3Modifier(Gameplay=<0 0 0>, Front End=<0 -200 0>))
  ]
  particleUpdateScript [
   &ParticleState()
   &DynamicMaterialParameters(
    Index 0 Param 1=&RandomRangeFloat(Minimum=0, Maximum=1)
    Index 0 Param 2=1
    Module.Param0WriteEnabled=true
    Module.Param1WriteEnabled=true
    Module.Param2WriteEnabled=true
    Module.Param3WriteEnabled=true
   )
   &ScaleMeshSize(Scale Factor=&Multiply_VectorByFloat(
     Vector=&MakeVector(
      X=&Lerp_Float(
       A=0.25
       B=1
       Alpha=&FloatFromCurve(FloatCurve=DataInterface() {propertyValues {Curve {keys [
            {arriveTangent=4.1785, interpMode=RCIM_Cubic, leaveTangent=4.1785, tangentMode=RCTM_User, time=0, value=0}
            {arriveTangent=0.9995, interpMode=RCIM_Cubic, leaveTangent=0.9995, tangentMode=RCTM_User, time=0.2223, value=0.6825}
            {arriveTangent=-0.0405, interpMode=RCIM_Cubic, leaveTangent=-0.0405, tangentMode=RCTM_User, time=1, value=1}
           ]}}})
      )
      Y=&Lerp_Float(
       A=1
       B=0.25
       Alpha=&FloatFromCurve(FloatCurve=DataInterface() {propertyValues {Curve {keys [
            {arriveTangent=4.1785, interpMode=RCIM_Cubic, leaveTangent=4.1785, tangentMode=RCTM_User, time=0, value=0}
            {arriveTangent=0.9995, interpMode=RCIM_Cubic, leaveTangent=0.9995, tangentMode=RCTM_User, time=0.2223, value=0.6825}
            {arriveTangent=-0.0405, interpMode=RCIM_Cubic, leaveTangent=-0.0405, tangentMode=RCTM_User, time=1, value=1}
           ]}}})
      )
      Z=&Lerp_Float(
       A=1
       B=0.25
       Alpha=&FloatFromCurve(FloatCurve=DataInterface() {propertyValues {Curve {keys [
            {arriveTangent=4.1785, interpMode=RCIM_Cubic, leaveTangent=4.1785, tangentMode=RCTM_User, time=0, value=0}
            {arriveTangent=0.9995, interpMode=RCIM_Cubic, leaveTangent=0.9995, tangentMode=RCTM_User, time=0.2223, value=0.6825}
            {arriveTangent=-0.0405, interpMode=RCIM_Cubic, leaveTangent=-0.0405, tangentMode=RCTM_User, time=1, value=1}
           ]}}})
      )
     )
     Float=&Lerp_Float(
      A=0
      B=1
      Alpha=&FloatFromCurve(FloatCurve=DataInterface() {propertyValues {Curve {keys [
           {arriveTangent=9.0725, interpMode=RCIM_Cubic, leaveTangent=9.0725, tangentMode=RCTM_User, time=0, value=0}
           {arriveTangent=0.7857, interpMode=RCIM_Cubic, leaveTangent=0.7857, time=0.1408, value=0.9847}
           {arriveTangent=-2.2398, interpMode=RCIM_Cubic, leaveTangent=-2.2398, tangentMode=RCTM_User, time=0.7, value=0.55}
           {arriveTangent=-0.0405, interpMode=RCIM_Cubic, leaveTangent=-0.0405, tangentMode=RCTM_User, time=1, value=0}
          ]}}})
     )
    ))
   &InheritSourceMovement(
    Applied Force Scale=<0.1 0.1 0.1>
    Applied Position Falloff Distance=100
    Applied Position Scale=<0.8 0.8 0.8>
    ApplyForce=true
    ApplyForceFalloff=false
    ApplyPosition=true
    ApplyPositionFalloff=false
    Module.LimitForceSpeed=false
   )
   &SolveForcesAndVelocity(Clamp Velocity=false, Limit Acceleration=false)
   &ScaleMeshSize(Scale Factor=&FN_FrontEndVector3Modifier(Gameplay=<1 1 1>, Front End=<0.8 0.8 0.8>)) {module=ScaleMeshSize001}
  ]
  rendererClasses [&NiagaraMeshRendererProperties]
  renderers [{rendererClass=&NiagaraMeshRendererProperties, rendererIndex=0}]
 }
 OuterCone {
  emitterUpdateScript [
   &EmitterState()
   &SpawnRate(
    SpawnRate=&FN_FrontEndFloatModifier(Gameplay=&Lerp_Float(A=0, B=8, Alpha=$User.ForwardDot), Front End=0)
    Use Spawn Probability=false
   )
  ]
  particleSpawnScript [
   /Niagara/Modules/Spawn/Initialization/V2/InitializeParticle(
    Color=&MakeLinearColorFromVectorAndFloat(
     "Vector (RGB)"=<1 1 1>
     "Float (Alpha)"=&Multiply_Float(A=&RandomRangeFloat(Minimum=0.5, Maximum=1), B=0.75)
    )
    Lifetime=0.35
    Mesh Scale=&Multiply_VectorByFloat(Vector=<1.25 1 1.3>, Float=0.8)
    Mesh Scale Mode=::NewEnumerator4
    Sprite Size Mode=::NewEnumerator3
    UsePositionOffset=false
    Write Material Random=true
   )
   &InitialMeshOrientation(
    Rotation=&FN_FrontEndVector3Modifier(Gameplay=<0.68 0 0>, Front End=<0.35 0 0>)
    Use Orientation Vector=true
    Use Rotation Vector=true
   )
  ]
  particleUpdateScript [
   &SkeletalMeshLocation(
    Apply Sampled Orientation to Mesh Particles=true
    Bone Sampling Mode=::NewEnumerator4
    Inherit Velocity=false
    MaskSpawnGroup=false
    Sampled Position Offset=<0 10 30>
    Sampled Position Offset Coordinate Space:ENiagaraMeshTransforms=::NewEnumerator3
    Skeletal Mesh=$User.SkeletalMesh
   )
   &ParticleState()
  ]
  rendererClasses [&NiagaraMeshRendererProperties]
  renderers [{rendererClass=&NiagaraMeshRendererProperties, rendererIndex=0}]
 }
}
summary {
 systemName=NS_Glider_Accumulate_Trail
 emitters [
  {emitterName=OuterCone, rendererClasses [&NiagaraMeshRendererProperties]}
  {emitterName=DroolPendulum, rendererClasses [&NiagaraMeshRendererProperties]}
  {emitterName=Drips, rendererClasses [&NiagaraMeshRendererProperties]}
  {emitterName=LargeDrips, rendererClasses [&NiagaraMeshRendererProperties]}
 ]
 userVariables [
  User.bIsFrontEnd {type=NiagaraBool}
  User.bIsFrontEnd {type=NiagaraFloat, defaultValue=0, description=""}
  User.bIsFrontEndPreview {type=NiagaraBool}
  User.bisFullyDeployed {type=NiagaraBool}
  User.Color {type=LinearColor, description="", defaultValue {struct=LinearColor, value=<rgba 0.01 0.01 0.01 1>}}
  User.ForwardDot {type=NiagaraFloat, defaultValue=0, description=""}
  User.RibbonIDOffset {type=NiagaraInt32}
  User.RightDot {type=NiagaraFloat, defaultValue=0, description=""}
  User.SkeletalMesh {
   type=NiagaraDataInterfaceSkeletalMesh
   defaultValue {
    struct=NiagaraExt_VariableValue_DataInterface
    value {dataInterface=&self:NiagaraDataInterfaceSkeletalMesh_0}
   }
  }
  User.SKM_Back {
   type=NiagaraDataInterfaceSkeletalMesh
   defaultValue {
    struct=NiagaraExt_VariableValue_DataInterface
    value {dataInterface=&self:NiagaraDataInterfaceSkeletalMesh_1}
   }
  }
  User.SKM_Tail {
   type=NiagaraDataInterfaceSkeletalMesh
   defaultValue {
    struct=NiagaraExt_VariableValue_DataInterface
    value {dataInterface=&self:NiagaraDataInterfaceSkeletalMesh_4}
   }
  }
  User.SKM_Top {
   type=NiagaraDataInterfaceSkeletalMesh
   defaultValue {
    struct=NiagaraExt_VariableValue_DataInterface
    value {dataInterface=&self:NiagaraDataInterfaceSkeletalMesh_2}
   }
  }
 ]
}
